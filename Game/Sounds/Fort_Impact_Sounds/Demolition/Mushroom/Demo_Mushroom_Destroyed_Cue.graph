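# Sound cue for the impact sound played when a mushroom is destroyed in demolition, routed to the impacts sound class with medium-object destruction attenuation.
# A quality-level switch sends the first two levels to a random pick of four waves through a default modulator.
# The last two levels pick randomly from two of those waves, with explicitly set pitch variation.
# A shared object-impact concurrency set limits simultaneous playback.
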SoundCue /Game/Sounds/Fort_Impact_Sounds/Demolition/Mushroom/Demo_Mushroom_Destroyed_Cue
root=QualityLevel_0
nodes [
 {id=Modulator_0, children [Random_2], properties {pitchMax=1.1, pitchMin=0.9}}
 {id=Modulator_1, children [Random_1]}
 {id=QualityLevel_0, children [Modulator_1,Modulator_1,Modulator_0,Modulator_0]}
 {id=Random_1, children [WavePlayer_3,WavePlayer_4,WavePlayer_5,WavePlayer_6], properties {weights [1,1,1,1]}}
 {id=Random_2, children [WavePlayer_7,WavePlayer_8], properties {weights [1,1]}}
 {id=WavePlayer_3, properties {soundWaveAssetPtr=&/Demo_Mushroom_Destroyed_04}}
 {id=WavePlayer_4, properties {soundWaveAssetPtr=&/Demo_Mushroom_Destroyed_03}}
 {id=WavePlayer_5, properties {soundWaveAssetPtr=&/Demo_Mushroom_Destroyed_02}}
 {id=WavePlayer_6, properties {soundWaveAssetPtr=&/Demo_Mushroom_Destroyed_01}}
 {id=WavePlayer_7, properties {soundWaveAssetPtr=&Demo_Mushroom_Destroyed_04}}
 {id=WavePlayer_8, properties {soundWaveAssetPtr=&Demo_Mushroom_Destroyed_02}}
]
properties {
 attenuationSettings=/Game/Sounds/Attenuation/Fort_Demolition_Destroy_Medium_Sized_Object_Att
 maxDistance=4000
 soundClassObject=&Fort_Impacts
 volumeMultiplier=1.2
 concurrencyOverrides {resolutionRule=PreventNew}
 concurrencySet [&CONC_Object_Impacts_FourMax]
}
