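# Sound cue for mushroom demolition impacts, built from two reused hay container search recordings.
# A Random node picks one recording with equal weight, an Enveloper fades its volume out over about 1.5 seconds, and a Modulator slightly varies pitch.
# It plays in the Fort_Impacts sound class with close-hit demolition crossfade attenuation and overrides concurrency to block new instances at the limit.
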SoundCue /Game/Sounds/Fort_Impact_Sounds/Demolition/Mushroom/Demo_Mushroom_Impact_1_8_Cue
root=Modulator_0
nodes [
 {
  id=Enveloper_0
  children [Random_12]
  properties {pitchCurve {editorCurveData {keys [{time=0}]}}, volumeCurve {editorCurveData {keys [<at 0 1 RCIM_Cubic>,<at 1.5429 -0.0334>]}}}
 }
 {id=Modulator_0, children [Enveloper_0], properties {pitchMax=1.07, pitchMin=0.92}}
 {id=Random_12, children [WavePlayer_54, WavePlayer_55], properties {weights [1,1]}}
 {id=WavePlayer_54, properties {soundWaveAssetPtr=/Game/Sounds/Foley_Loot/Containers/Hay/hay_search_02}}
 {id=WavePlayer_55, properties {soundWaveAssetPtr=/Game/Sounds/Foley_Loot/Containers/Hay/hay_search_01}}
]
properties {
 attenuationSettings=/Game/Sounds/Attenuation/Fort_Demolition_Crossfade_Global_Hit_Close_Att
 bOverrideConcurrency=true
 maxDistance=4800
 soundClassObject=&Fort_Impacts
 volumeMultiplier=0.566
 concurrencyOverrides {resolutionRule=PreventNew}
}
